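; Parent Blueprint for composite encounter volumes that reports an encounter's outcome to its State Tree.
; FinishEncounter gets the EncounterStateTreeComponent from EncounterGameplayVolume and sends it an Encounter.Event.Success or Encounter.Event.Failure tag event, based on IsSuccess, which defaults to true.
; BeginPlay, Tick, BeginOverlap and the construction script are empty, and nothing in this Blueprint calls FinishEncounter.
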(event EventBeginPlay)

(event Collision|EventActorBeginOverlap
  (params (OtherActor :type Actor%20Object%20Reference)))

(event EventTick
  (params (DeltaSeconds :type Float%20%28single-precision%29)))

(fn FinishEncounter
  (params (IsSuccess :type Boolean :default true))
  (bind _encounterstatetreecomponent (Variables|EncounterGameplayVolume|GetEncounterStateTreeComponent))
  (if IsSuccess
    (Gameplay|StateTree|SendStateTreeEvent _encounterstatetreecomponent (Utilities|Struct|MakeStateTreeEvent "(TagName=\"Encounter.Event.Success\")") :!class /Script/GameplayStateTreeModule.StateTreeComponent)
    (else
      (Gameplay|StateTree|SendStateTreeEvent _encounterstatetreecomponent (Utilities|Struct|MakeStateTreeEvent "(TagName=\"Encounter.Event.Failure\")") :!class /Script/GameplayStateTreeModule.StateTreeComponent))))

(fn ConstructionScript)
