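; Parent spawner Blueprint for NPC prefab actors in Fort Athena LW encounters that tags each pawn it spawns.
; On BeginPlay it registers a custom event as the callback for when an actor spawns.
; That callback adds a loose gameplay tag from its NPC pawn tag variable to the spawned actor and prints a green debug message on success.
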(event EventBeginPlay
  (CallFunction|CallwhenActorSpawned (delegate Custom|ActorSpawned)))

(event Custom|ActorSpawned
  (params (SpawnedActor :type Actor%20Object%20Reference))
  (bind _returnvalue (Ability|GameplayEffect|AddLooseGameplayTags SpawnedActor (Variables|Default|GetNPC_Pawn_Tag) true))
  (if _returnvalue
    (Development|PrintString "Tag added on villager pawn" true true "(R=0.000000,G=1.000000,B=0.049684,A=1.000000)" 3.0)))

(fn ConstructionScript)
